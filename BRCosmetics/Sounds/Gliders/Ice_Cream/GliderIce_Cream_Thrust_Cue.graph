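# Sound cue for the Ice Cream cosmetic glider's thrust sound, built from three thrust-start waves.
# A random node picks one wave at equal weight and feeds a modulator that slightly varies pitch on each play.
# The cue lowers overall pitch and volume and uses shared glider thrust attenuation and concurrency.
# It references the glider landing sound class and a glider loop audio properties sheet, though the graph has no looping node.
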SoundCue /BRCosmetics/Sounds/Gliders/Ice_Cream/GliderIce_Cream_Thrust_Cue
root=Modulator_0
nodes [
 {id=Modulator_0, children [Random_0], properties {pitchMax=1.1, pitchMin=0.9}}
 {id=Random_0, children [WavePlayer_0,WavePlayer_1,WavePlayer_3], properties {weights [1,1,1]}}
 {id=WavePlayer_0, properties {soundWaveAssetPtr=&/Glider_Ice_Cream_Thrust_Start_01}}
 {id=WavePlayer_1, properties {soundWaveAssetPtr=&/Glider_Ice_Cream_Thrust_Start_02}}
 {id=WavePlayer_3, properties {soundWaveAssetPtr=&/Glider_Ice_Cream_Thrust_Start_03}}
]
properties {
 attenuationSettings=&Fort_Player_Glider_Thrust_Athena_Att
 audioPropertiesSheet=&APS_Cosmetic_Glider_Loop
 maxDistance=12500
 pitchMultiplier=0.7
 soundClassObject=&Fort_Gliders_Land
 volumeMultiplier=0.25
 concurrencySet [&Athena_Glider_Thrust_Conc]
}
